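# Niagara trail effect for the CozyMechanic glider chute, with a frosty, snowy look.
# Emitters layer socket-attached particles, frost sprites, snow and streak meshes, plus anim-trail wing ribbons and wind lines.
# System-age curves and spawn groups separate an early burst phase from the sustained trail, while timeline values and user ForwardDot/RightDot inputs modulate spawning, motion and fades.
# Front-end modifiers retune or suppress parts for preview, and time-of-day inputs adjust color and alpha.
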NiagaraSystem /BRCosmetics/Gadgets/Assets/VinderTech_GliderChute/Glider_CozyMechanic/FX/NS_Glider_CozyMechanic_Trail_01
dependencies {
 usedDataInterfaces [
  &NiagaraDataInterfaceCurve
  &NiagaraDataInterfaceColorCurve
  &NiagaraDataInterfaceSocketReader
  &NiagaraDataInterfaceArrayFloat
 ]
 usedDynamicInputs [
  &SelectFloatFromArray
  &NormalizeVector
  &TransformVector
  &FNTimeOfDayColor
 ]
 usedModules [
  &SystemState
  &self:SystemSpawnScript.SystemScriptSource.SystemScriptGraph.NiagaraNodeAssignment_2.SetVariables_#1
  &Timeline
  &FN_Integrator
  &self:SystemSpawnScript.SystemScriptSource.SystemScriptGraph.NiagaraNodeAssignment_3.SetVariables_#2
  &self:SystemSpawnScript.SystemScriptSource.SystemScriptGraph.NiagaraNodeAssignment_0.SetVariables_#3
 ]
}
emitters {
 CPU_particles {
  emitterUpdateScript [
   &EmitterState(Delay First Loop Only=true, Life Cycle Mode=::NewEnumerator1, Loop Delay=1.35, Loop Duration=1, UseLoopDelay=true)
   &SpawnRate(
    SpawnRate=&Multiply_Float(
     A=&FloatFromCurve(FloatCurve=DataInterface() {propertyValues {Curve {keys [<at 2.7 0>,<at 3 1>]}}}, CurveIndex=$System.Age, Scale Curve=80)
     B=&FN_FrontEndFloatModifier(Gameplay=&ClampFloat(Float=$User.ForwardDot, Min=0.2, Max=0.6), Front End=1)
    )
    Use Spawn Probability=false
   )
   &SpawnRate(
    SpawnGroup=1
    SpawnRate=&FloatFromCurve(
     FloatCurve=DataInterface() {propertyValues {Curve {keys [<at 1.3 1>,<at 1.7 0>]}}}
     CurveIndex=$System.Age
     Scale Curve=&FN_FrontEndFloatModifier(Gameplay=180, Front End=0)
    )
    Use Spawn Probability=false
   ) {module=SpawnRate001}
  ]
  particleSpawnScript [
   &InitializeParticle(
    Color=<rgba 1.9412 2.8115 5 1>
    Lifetime=&Lerp_Float(
     A=&RandomRangeFloat(Minimum=0.7, Maximum=1.1)
     B=&RandomRangeFloat(Minimum=0.4, Maximum=0.7)
     Alpha=$System.Timeline.Time
    )
    Sprite Rotation Mode=::NewEnumerator2
    Sprite Size Mode=::NewEnumerator1
    Uniform Sprite Size Max=15
    UsePositionOffset=false
    Write Material Random=true
   )
   &SocketLocation(
    SamplePosOffset=true
    Sampled Position Offset=&RandomRangeVector(Minimum=<0 65 0>, Maximum=<0 65 0>)
    Socket Reader=$System.Socket Reader
    Socket Sampling Mode=::NewEnumerator3
    VelocitySpeedLimit=false
   )
   &SphereLocation(
    MaskSpawnGroup=false
    Non Uniform Scale=<1 1 0.7>
    Override Seed=false
    Sphere Radius=&Multiply_Float(A=35, B=&ClampFloat(Float=$User.ForwardDot, Min=0.8))
    Surface Only=false
    UseNonUniformScale=true
    UseOffset=false
    UseSphereOrientation=false
   )
   &AddVelocity(
    Velocity=&RandomRangeVector(Minimum=<-40 200 30>, Maximum=<40 300 60>)
    Velocity Speed Scale=&MergeHelper_FloatSwitch(
     0=&Lerp_Float(A=1, B=1.7, Alpha=$System.Timeline.Time)
     1=3
     2=?
     3=?
     Index=&SpawnGroup()
     Num=::NewEnumerator0
    )
   )
   &CameraOffset(Camera Offset Amount=10)
   /SetVariables(Particles.VisibilityTag=&MakeIntFromFloat(
     Float=&UniformAOrBFloat(A=0, Distribution Weight=0.3, Evaluation Type=::NewEnumerator0, Recalculate Random Each Loop=true, Override Seed=false)
     Conversion Mode=::NewEnumerator1
    )) {isSetParameters=true}
   &DynamicMaterialParameters(
    Index 0 Param 1=0.5
    Index 0 Param 2=&FloatFromCurve(FloatCurve=DataInterface() {propertyValues {Curve {keys [{arriveTangent=-12.4095, interpMode=RCIM_Cubic, leaveTangent=-12.4095, tangentMode=RCTM_User, time=0.5, value=2}, <at 1 0 RCIM_Cubic>]}}})
    Module.Param0WriteEnabled=true
    Module.Param1WriteEnabled=true
    Module.Param2WriteEnabled=true
    Module.Param3WriteEnabled=true
   ) {module=DynamicMaterialParameters001}
  ]
  particleUpdateScript [
   /SetVariables(Particles.Sprite02_Size:Vector2f=&Vector2DFromFloat(Value=&Multiply_Float(
      A=&RandomRangeFloat(Minimum=10, Maximum=20)
      B=&FloatFromCurve(FloatCurve=DataInterface() {propertyValues {Curve {keys [<at 0 0 RCIM_Cubic>,<at 0.1351 0.878 RCIM_Cubic>,<at 1 0 RCIM_Cubic>]}}})
     ))) {isSetParameters=true}
   &AccelerationForce(Acceleration=&MakeVector(X=0, Y=0, Z=300))
   &ScaleVelocity(Velocity Scale=&Multiply_VectorByFloat(
     Vector=<1 1 1>
     Float=&FN_FrontEndFloatModifier(Gameplay=&ClampFloat(Float=$User.ForwardDot, Min=0.5), Front End=1)
    ))
   &SolveForcesAndVelocity(Clamp Velocity=false, Limit Acceleration=false)
   &ParticleState()
   &ScaleSpriteSize(Scale Factor=&Vector2DFromFloat(Value=&FloatFromCurve(
      FloatCurve=DataInterface() {propertyValues {Curve {keys [<at 0 0 RCIM_Cubic>,<at 0.1351 0.878 RCIM_Cubic>,<at 1 0 RCIM_Cubic>]}}}
      Scale Curve=0.8
     ))) {module=ScaleSpriteSize001}
   &SpriteRotationRate(Rotation Rate=&RandomRangeFloat(Minimum=-1, Maximum=1))
   &ScaleColor(
    Scale Alpha=&Multiply_Float(
     A=&RandomRangeFloat(Minimum=0.4, Maximum=0.8)
     B=&FloatFromCurve(
      FloatCurve=DataInterface() {propertyValues {Curve {keys [
          {arriveTangent=4.1785, interpMode=RCIM_Cubic, leaveTangent=4.1785, tangentMode=RCTM_User, time=0, value=0}
          {arriveTangent=0.9995, interpMode=RCIM_Cubic, leaveTangent=0.9995, tangentMode=RCTM_User, time=0.2223, value=0.6825}
          {arriveTangent=-0.0405, interpMode=RCIM_Cubic, leaveTangent=-0.0405, tangentMode=RCTM_User, time=1, value=1}
         ]}}}
      Scale Curve=&FNTimeOfDayScalar(Night Scale=1, Morning Scale=0.2, Evening Scale=0.4)
     )
    )
    Scale RGB=&VectorFromFloat(Value=&FloatFromCurve(FloatCurve=DataInterface() {propertyValues {Curve {keys [<at 1.3 3>,<at 2 1>]}}}, CurveIndex=$System.Age))
    ScaleA=true
    ScaleRGB=true
   )
  ]
  rendererClasses [&NiagaraSpriteRendererProperties]
  renderers [
   {rendererClass=&NiagaraSpriteRendererProperties}
   {rendererClass=&NiagaraSpriteRendererProperties, rendererIndex=1}
  ]
 }
 Frost {
  emitterUpdateScript [
   &EmitterState(Delay First Loop Only=true, Life Cycle Mode=::NewEnumerator1, Loop Delay=1.35, UseLoopDelay=true)
   &SpawnRate(
    SpawnRate=&FloatFromCurve(
     FloatCurve=DataInterface() {propertyValues {Curve {keys [<at 1.3 1>,<at 1.95 0>]}}}
     CurveIndex=$System.Age
     Scale Curve=&FN_FrontEndFloatModifier(Gameplay=55, Front End=0)
    )
    Use Spawn Probability=false
   ) {module=SpawnRate001}
   &SpawnBurst_Instantaneous(
    Spawn Count=&MakeIntFromFloat(Float=&FN_FrontEndFloatModifier(Gameplay=10, Front End=0), Conversion Mode=::NewEnumerator1)
    Spawn Group=1
    Spawn Time=1.2
    Use Spawn Probability=false
    UseLoopCountLimit=true
   )
   &SpawnRate(
    SpawnGroup=2
    SpawnRate=&Multiply_Float(
     A=&FloatFromCurve(FloatCurve=DataInterface() {propertyValues {Curve {keys [<at 2.8 0>,<at 3 1>]}}}, CurveIndex=$System.Age, Scale Curve=20)
     B=&FN_FrontEndFloatModifier(Gameplay=&ClampFloat(Float=$User.ForwardDot, Min=0.3), Front End=0.7)
    )
    Use Spawn Probability=false
   )
  ]
  particleSpawnScript [
   &InitializeParticle(
    Color=&MergeHelper_ColorSwitch(
     0=&MakeLinearColorFromVectorAndFloat(
      "Vector (RGB)"=&MakeVectorFromLinearColorRGB(Color=<rgba 0.88 0.9947 1 1>)
      "Float (Alpha)"=&Multiply_Float(
       A=&RandomRangeFloat(Minimum=0.03, Maximum=0.06)
       B=&FN_FrontEndFloatModifier(Gameplay=1, Front End=0.5)
      )
     )
     1=<rgba 0.2708 0.3275 0.4 0.2>
     2=<rgba 0.4492 0.6108 0.8177 0.2>
     3=?
     Index=&SpawnGroup()
     Num=::NewEnumerator1
    )
    Lifetime=&MergeHelper_FloatSwitch(
     0=0.35
     1=&RandomRangeFloat(Minimum=0.7, Maximum=1)
     2=&RandomRangeFloat(Minimum=0.7, Maximum=1.2)
     3=?
     Index=&SpawnGroup()
     Num=::NewEnumerator1
    )
    Sprite Rotation Mode=::NewEnumerator2
    Sprite Size Mode=::NewEnumerator3
    Sprite UV Mode=::NewEnumerator3
    Uniform Sprite Size=&MergeHelper_FloatSwitch(
     0=&RandomRangeFloat(Minimum=50, Maximum=55)
     1=&RandomRangeFloat(Minimum=120, Maximum=150)
     2=&RandomRangeFloat(Minimum=120, Maximum=150)
     3=?
     Index=&SpawnGroup()
     Num=::NewEnumerator1
    )
    UsePositionOffset=false
    Write Material Random=true
   )
   &CameraOffset(Camera Offset Amount=&MergeHelper_FloatSwitch(0=5, 1=10, 2=10, 3=?, Index=&SpawnGroup(), Num=::NewEnumerator1))
   &SocketLocation(
    SamplePosOffset=true
    Sampled Position Offset=&MergeHelper_VectorSwitch(0=<0 50 0>, 1=<0 0 -10>, 2=<0 -120 -10>, 3=?, Index=&SpawnGroup(), Num=::NewEnumerator1)
    Socket Reader=$System.Socket Reader
    Socket Sampling Mode=::NewEnumerator3
    VelocitySpeedLimit=false
   )
   &TorusLocation(
    Align Torus With Axis=false
    Handle Radius=0
    Large Radius:NiagaraFloat=&MergeHelper_FloatSwitch(
     0=0
     1=&Multiply_Float(A=30, B=&ClampFloat(Float=$User.ForwardDot, Min=0.3))
     2=&Multiply_Float(A=30, B=&ClampFloat(Float=$User.ForwardDot, Min=0.3, Max=0.7))
     3=?
     Index=&SpawnGroup()
     Num=::NewEnumerator1
    )
    MaskSpawnGroup=false
    Module.Override Randomness=false
   )
   &AddVelocityFromPoint(
    Module.Use Falloff Distance=false
    Origin Offset=&MergeHelper_VectorSwitch(0=<0 0 0>, 1=<0 0 -20>, 2=<0 -150 0>, 3=?, Index=&SpawnGroup(), Num=::NewEnumerator1)
    Velocity Strength=&MergeHelper_FloatSwitch(
     0=0
     1=250
     2=&FN_FrontEndFloatModifier(Gameplay=200, Front End=100)
     3=?
     Index=&SpawnGroup()
     Num=::NewEnumerator1
    )
   )
   &AddVelocity(
    Velocity=&Multiply_VectorByFloat(
     Vector=&MergeHelper_VectorSwitch(
      0=&FN_FrontEndVector3Modifier(Gameplay=<0 90 15>, Front End=<0 90 30>)
      1=<0 0 0>
      2=<0 100 0>
      3=?
      Index=&SpawnGroup()
      Num=::NewEnumerator1
     )
     Float=&FloatFromCurve(
      FloatCurve=DataInterface() {propertyValues {Curve {keys [{arriveTangent=10.0357, interpMode=RCIM_Cubic, leaveTangent=10.0357, tangentMode=RCTM_User, time=0.2, value=0}, <at 0.6 1 RCIM_Cubic>]}}}
      CurveIndex=&RandomRangeFloat(Minimum=0.7, Maximum=1)
     )
    )
    Velocity Speed Scale=&MergeHelper_FloatSwitch(
     0=6
     1=0
     2=&FN_FrontEndFloatModifier(Gameplay=1, Front End=0.6)
     3=?
     Index=&SpawnGroup()
     Num=::NewEnumerator1
    )
   )
   &AddVelocityInCone(
    Cone Angle=10
    Cone Axis=<0 1 0>
    Module.Use Velocity Falloff On Cone Axis=false
    Velocity Strength=&MergeHelper_FloatSwitch(0=0, 1=0, 2=300, 3=?, Index=&SpawnGroup(), Num=::NewEnumerator1)
   ) {module=AddVelocityInCone001}
   /SetVariables(Particles.VisibilityTag=&SpawnGroup()) {isSetParameters=true}
   /SetVariables(Particles.DynamicLanding:Vector4f=&MakeVector4(X=1, Y=1, Z=1, W=1)) {isSetParameters=true}
  ]
  particleUpdateScript [
   &ParticleState()
   &AccelerationForce(Acceleration=&MergeHelper_VectorSwitch(
     0=<0 0 0>
     1=<0 0 0>
     2=&MakeVector(X=0, Y=0, Z=300)
     3=?
     Index=&SpawnGroup()
     Num=::NewEnumerator1
    ))
   &GravityForce(Gravity=&MergeHelper_VectorSwitch(
     0=<0 0 0>
     1=&RandomRangeVector(Minimum=<0 0 -600>, Maximum=<0 0 -600>)
     2=<0 0 0>
     3=?
     Index=&SpawnGroup()
     Num=::NewEnumerator1
    )) {module=GravityForce001}
   &Drag(
    Drag=&MergeHelper_FloatSwitch(0=0, 1=1, 2=0, 3=?, Index=&SpawnGroup(), Num=::NewEnumerator1)
    Use Linear Drag=true
    Use Rotational Drag=false
   )
   &ScaleVelocity(Velocity Scale=&Multiply_VectorByFloat(
     Vector=<1 1 1>
     Float=&MergeHelper_FloatSwitch(
      0=1
      1=1
      2=&FN_FrontEndFloatModifier(Gameplay=&ClampFloat(Float=$User.ForwardDot, Min=0.5), Front End=1)
      3=?
      Index=&SpawnGroup()
      Num=::NewEnumerator1
     )
    ))
   &SolveForcesAndVelocity(Clamp Velocity=false, Limit Acceleration=false)
   /Niagara/Modules/Update/SubUV/SubUVAnimation(Number Of Frames=63, SubUV Lookup Index=&Multiply_Float(A=$Particles.NormalizedAge, B=0.34))
   &ScaleSpriteSize(
    Scale Sprite Size Mode=::NewEnumerator0
    Uniform Scale Factor=&MergeHelper_FloatSwitch(
     0=&FloatFromCurve(
      FloatCurve=DataInterface() {propertyValues {Curve {keys [
          {arriveTangent=0.6687, interpMode=RCIM_Cubic, leaveTangent=0.6687, tangentMode=RCTM_User, time=0, value=0.4}
          {tangentMode=RCTM_User, time=1, value=1}
         ]}}}
      Scale Curve=2
     )
     1=&FloatFromCurve(FloatCurve=DataInterface() {propertyValues {Curve {keys [
          {arriveTangent=4.1785, interpMode=RCIM_Cubic, leaveTangent=4.1785, tangentMode=RCTM_User, time=0, value=0}
          {arriveTangent=0.9995, interpMode=RCIM_Cubic, leaveTangent=0.9995, tangentMode=RCTM_User, time=0.2223, value=0.6825}
          {arriveTangent=-0.0405, interpMode=RCIM_Cubic, leaveTangent=-0.0405, tangentMode=RCTM_User, time=1, value=1}
         ]}}})
     2=&FloatFromCurve(FloatCurve=DataInterface() {propertyValues {Curve {keys [
          {arriveTangent=4.1785, interpMode=RCIM_Cubic, leaveTangent=4.1785, tangentMode=RCTM_User, time=0, value=0}
          {arriveTangent=0.9995, interpMode=RCIM_Cubic, leaveTangent=0.9995, tangentMode=RCTM_User, time=0.2223, value=0.6825}
          {arriveTangent=-0.0405, interpMode=RCIM_Cubic, leaveTangent=-0.0405, tangentMode=RCTM_User, time=1, value=1}
         ]}}})
     3=?
     Index=&SpawnGroup()
     Num=::NewEnumerator1
    )
   )
   &DynamicMaterialParameters(
    Index 0 Param 1=0.1
    Module.Index0Float0Write=true
    Module.Index0Float1Write=true
    Module.Index0Float2Write=true
    Module.Index0Float3Write=true
   )
   &ScaleColor(
    Scale Alpha=&MergeHelper_FloatSwitch(
     0=&FloatFromCurve(FloatCurve=DataInterface() {propertyValues {Curve {keys [
          {arriveTangent=0.9664, interpMode=RCIM_Cubic, leaveTangent=0.9664, tangentMode=RCTM_User, time=0, value=0}
          <at 0.429 1 RCIM_Cubic>
          {arriveTangent=-0.669, interpMode=RCIM_Cubic, leaveTangent=-0.669, tangentMode=RCTM_User, time=1, value=0}
         ]}}})
     1=&Multiply_Float(
      A=&FloatFromCurve(
       FloatCurve=DataInterface() {propertyValues {Curve {keys [
           {interpMode=RCIM_Cubic, tangentMode=RCTM_User, time=0, value=0}
           {interpMode=RCIM_Cubic, leaveTangent=-2.6773, tangentMode=RCTM_Break, time=0.3, value=1}
           {tangentMode=RCTM_User, time=1, value=0}
          ]}}}
       Scale Curve=&RandomRangeFloat(Minimum=0.5, Maximum=1)
      )
      B=&FNTimeOfDayScalar(Night Scale=1, Morning Scale=0.2, Evening Scale=0.2)
     )
     2=&Multiply_Float(
      A=&FloatFromCurve(
       FloatCurve=DataInterface() {propertyValues {Curve {keys [
           {interpMode=RCIM_Cubic, tangentMode=RCTM_User, time=0, value=0}
           {interpMode=RCIM_Cubic, leaveTangent=-2.6773, tangentMode=RCTM_Break, time=0.3, value=1}
           {tangentMode=RCTM_User, time=1, value=0}
          ]}}}
       Scale Curve=&RandomRangeFloat(Minimum=0.5, Maximum=1)
      )
      B=&FNTimeOfDayScalar(Night Scale=1, Morning Scale=0.2, Evening Scale=0.2)
     )
     3=?
     Index=&SpawnGroup()
     Num=::NewEnumerator1
    )
    ScaleA=true
    ScaleRGB=true
   )
   &ScaleColor(Color Scale:LinearColor=$System.TOD Color, Scale Mode=3) {module=ScaleColor002}
  ]
  rendererClasses [&NiagaraSpriteRendererProperties]
  renderers [
   {rendererClass=&NiagaraSpriteRendererProperties}
   {rendererClass=&NiagaraSpriteRendererProperties, rendererIndex=1}
   {rendererClass=&NiagaraSpriteRendererProperties, rendererIndex=2}
  ]
 }
 LongTrail {
  emitterUpdateScript [
   &EmitterState(Delay First Loop Only=true, Life Cycle Mode=::NewEnumerator1, Loop Delay=2.55, Loop Duration=1, UseLoopDelay=true)
   &SpawnRate(
    SpawnRate=&Multiply_Float(
     A=28
     B=&FN_FrontEndFloatModifier(Gameplay=&ClampFloat(Float=$User.ForwardDot), Front End=1)
    )
    Use Spawn Probability=false
   )
  ]
  particleSpawnScript [
   &FN_SetSpawnGroup(Number of Spawn Groups=2)
   &InitializeParticle(
    Lifetime Max=0.8
    Lifetime Min=0.3
    Lifetime Mode=::NewEnumerator1
    Mesh Scale=&RandomRangeVector(Minimum=<3 3 3>, Maximum=<3 3 5>)
    Mesh Scale Mode=::NewEnumerator4
    UsePositionOffset=false
    Write Material Random=true
   )
   &SocketLocation(
    SamplePosOffset=true
    Sampled Position Offset=&MergeHelper_VectorSwitch(
     0=&Add_Vector(A=&RandomRangeVector(Minimum=<0 -50 0>, Maximum=<10 0 0>), B=<-30 -10 -10>)
     1=&Add_Vector(A=&RandomRangeVector(Minimum=<0 -50 0>, Maximum=<10 0 0>), B=<20 -10 -10>)
     2=?
     3=?
     Index=&SpawnGroup()
     Num=::NewEnumerator0
    )
    Socket Reader=$System.Socket Reader
    Socket Sampling Mode=::NewEnumerator3
    VelocitySpeedLimit=false
   )
   &InitialMeshOrientation(Mesh Orientation Mode=::NewEnumerator2, Rotation=<-0.25 0 0>, Use Rotation Vector=true)
   &AddVelocity(Velocity=<0 350 0>)
   &InitialMeshOrientation(
    Mesh Orientation Mode=::NewEnumerator2
    Rotation=&MergeHelper_VectorSwitch(
     0=&RandomRangeVector(Minimum=<0 0 0.03>, Maximum=<0 0 0.07>)
     1=<0 0.5 0>
     2=?
     3=?
     Index=&SpawnGroup()
     Num=::NewEnumerator0
    )
    Use Rotation Vector=true
   ) {module=InitialMeshOrientation002}
   &InitialMeshOrientation(
    Mesh Orientation Mode=::NewEnumerator2
    Rotation=&MergeHelper_VectorSwitch(
     0=&RandomRangeVector(Minimum=<0 0.15 0>, Maximum=<0 0.1 0>)
     1=&RandomRangeVector(Minimum=<0 0 -0.03>, Maximum=<0 0 -0.1>)
     2=?
     3=?
     Index=&SpawnGroup()
     Num=::NewEnumerator0
    )
    Use Rotation Vector=true
   ) {module=InitialMeshOrientation003}
   &InitialMeshOrientation(
    Mesh Orientation Mode=::NewEnumerator2
    Rotation=&MergeHelper_VectorSwitch(
     0=&RandomRangeVector(Minimum=<0 0 0>, Maximum=<0 0 0>)
     1=&RandomRangeVector(Minimum=<0 -0.15 0>, Maximum=<0 -0.1 0>)
     2=?
     3=?
     Index=&SpawnGroup()
     Num=::NewEnumerator0
    )
    Use Rotation Vector=true
   ) {module=InitialMeshOrientation001}
  ]
  particleUpdateScript [
   /SetVariables(Particles.Mesh_Low_Color:LinearColor=&MakeLinearColorFromVectorAndFloat(
     "Vector (RGB)"=&Multiply_VectorByFloat(
      Vector=&MakeVectorFromLinearColorRGB(Color=&MakeLinearColorFromVectorAndFloat(
        "Vector (RGB)"=&FN_FrontEndVector3Modifier(Gameplay=<0.7 1 1>, Front End=<0.4 0.8 1>)
        "Float (Alpha)"=1
       ))
      Float=&Multiply_Float(A=&FN_FrontEndFloatModifier(Gameplay=1, Front End=1.1), B=1)
     )
     "Float (Alpha)"=&Multiply_Float(
      A=&FloatFromCurve(
       FloatCurve=DataInterface() {propertyValues {Curve {keys [
           {interpMode=RCIM_Cubic, tangentMode=RCTM_User, time=0, value=0}
           {interpMode=RCIM_Cubic, tangentMode=RCTM_User, time=0.5, value=1}
           {tangentMode=RCTM_User, time=1, value=0}
          ]}}}
       Scale Curve=&Multiply_Float(
        A=&RandomRangeFloat(Minimum=0.5, Maximum=1)
        B=&FN_FrontEndFloatModifier(Gameplay=1, Front End=1)
       )
      )
      B=&FNTimeOfDayScalar(Night Scale=1, Morning Scale=0.7, Evening Scale=0.7)
     )
    )) {isSetParameters=true}
   &DynamicMaterialParameters(
    Index 0 Param 1=&Multiply_Float(
     A=&RandomRangeFloat(Minimum=1, Maximum=1.2)
     B=&FloatFromCurve(FloatCurve=DataInterface() {propertyValues {Curve {keys [
          {arriveTangent=4.1785, interpMode=RCIM_Cubic, leaveTangent=4.1785, tangentMode=RCTM_User, time=0, value=0}
          {arriveTangent=0.9995, interpMode=RCIM_Cubic, leaveTangent=0.9995, tangentMode=RCTM_User, time=0.2223, value=0.6825}
          {arriveTangent=-0.0405, interpMode=RCIM_Cubic, leaveTangent=-0.0405, tangentMode=RCTM_User, time=1, value=1}
         ]}}})
    )
    Index 0 Param 2=&RandomRangeFloat(Minimum=0, Maximum=0.8)
    Index 0 Param 3=&RandomRangeFloat(Minimum=0.5, Maximum=1)
    Index 0 Param 4=&RandomRangeFloat(Minimum=0.7, Maximum=1.2)
    Module.Param0WriteEnabled=true
    Module.Param1WriteEnabled=true
    Module.Param2WriteEnabled=true
    Module.Param3WriteEnabled=true
   )
   &SolveForcesAndVelocity(Clamp Velocity=false, Limit Acceleration=false)
   &ParticleState()
   &ScaleColor(
    Scale Alpha=&Multiply_Float(
     A=&FloatFromCurve(
      FloatCurve=DataInterface() {propertyValues {Curve {keys [
          {interpMode=RCIM_Cubic, tangentMode=RCTM_User, time=0, value=0}
          {interpMode=RCIM_Cubic, tangentMode=RCTM_User, time=0.5, value=1}
          {tangentMode=RCTM_User, time=1, value=0}
         ]}}}
      Scale Curve=&RandomRangeFloat(Minimum=0.4, Maximum=0.6)
     )
     B=&FNTimeOfDayScalar(Night Scale=0.3, Morning Scale=0.3, Day Scale=0.5, Evening Scale=0.15)
    )
    Scale RGB=&Multiply_VectorByFloat(
     Vector=&Lerp_Vector(
      A=&MakeVectorFromLinearColorRGB(Color=<rgba 0.7292 0.9097 1 1>)
      B=<1 1 1>
      Alpha=&FN_FrontEndFloatModifier(Gameplay=1, Front End=0)
     )
     Float=1
    )
    ScaleA=true
    ScaleRGB=true
   )
   &ScaleMeshSize(Scale Factor=&MakeVector(
     X=0.7
     Y=0.7
     Z=&FloatFromCurve(
      FloatCurve=DataInterface() {propertyValues {Curve {keys [
          {arriveTangent=4.1785, interpMode=RCIM_Cubic, leaveTangent=4.1785, tangentMode=RCTM_User, time=0, value=0.5}
          {arriveTangent=0.9995, interpMode=RCIM_Cubic, leaveTangent=0.9995, tangentMode=RCTM_User, time=0.2, value=0.75}
          {arriveTangent=-0.0405, interpMode=RCIM_Cubic, leaveTangent=-0.0405, tangentMode=RCTM_User, time=1, value=1.2}
         ]}}}
      Scale Curve=0.7
     )
    ))
  ]
  rendererClasses [&NiagaraMeshRendererProperties]
  renderers [{rendererClass=&NiagaraMeshRendererProperties, rendererIndex=0}]
 }
 Snow_Idle {
  emitterUpdateScript [
   &EmitterState(Delay First Loop Only=true, Life Cycle Mode=::NewEnumerator1, Loop Delay=1.35, Loop Duration=4, UseLoopDelay=true)
   &SpawnRate(
    SpawnRate=&FloatFromCurve(
     FloatCurve=DataInterface() {propertyValues {Curve {keys [<at 1.3 1>,<at 2 0>]}}}
     CurveIndex=$System.Age
     Scale Curve=&FN_FrontEndFloatModifier(Gameplay=70, Front End=0)
    )
    Use Spawn Probability=false
   ) {module=SpawnRate001}
   &SpawnBurst_Instantaneous(
    Spawn Count=&MakeIntFromFloat(Float=&FN_FrontEndFloatModifier(Gameplay=15, Front End=0))
    Spawn Group=1
    Spawn Time=1.2
    Use Spawn Probability=false
    UseLoopCountLimit=true
   )
   &SpawnRate(
    SpawnGroup=2
    SpawnRate=&Multiply_Float(
     A=&FloatFromCurve(FloatCurve=DataInterface() {propertyValues {Curve {keys [<at 2.5 0>,<at 2.6 1>]}}}, CurveIndex=$System.Age, Scale Curve=20)
     B=&FN_FrontEndFloatModifier(Gameplay=1, Front End=1)
    )
    Use Spawn Probability=false
   )
   &SpawnRate(
    SpawnGroup=3
    SpawnRate=&Multiply_Float(
     A=&FloatFromCurve(FloatCurve=DataInterface() {propertyValues {Curve {keys [<at 2.5 0>,<at 2.6 1>]}}}, CurveIndex=$System.Age, Scale Curve=30)
     B=&FN_FrontEndFloatModifier(Gameplay=&ClampFloat(Float=$User.ForwardDot), Front End=1)
    )
    Use Spawn Probability=false
   ) {module=SpawnRate002}
   &SpawnRate(
    SpawnGroup=4
    SpawnRate=&Multiply_Float(
     A=&FloatFromCurve(FloatCurve=DataInterface() {propertyValues {Curve {keys [<at 2.5 0>,<at 2.6 1>]}}}, CurveIndex=$System.Age, Scale Curve=30)
     B=&FN_FrontEndFloatModifier(Gameplay=&ClampFloat(Float=$User.ForwardDot), Front End=1)
    )
    Use Spawn Probability=false
   ) {module=SpawnRate003}
  ]
  particleSpawnScript [
   &InitializeParticle(
    Color=&MultiplyLinearColor(
     Linear Color=&LerpLinearColors(
      StartColor=<rgba 0.1337 0.2212 0.276 1>
      EndColor=<rgba 0.1337 0.2212 0.276 1>
      LerpFactor=&FN_FrontEndFloatModifier(Gameplay=0, Front End=1)
     )
     Scale Factor=&MakeVector4_FromFloat(Float=&FN_FrontEndFloatModifier(Gameplay=1, Front End=1))
    )
    Lifetime=&MergeHelper_FloatSwitch(
     0=0.35
     1=&RandomRangeFloat(Minimum=0.3, Maximum=0.5)
     2=0.7
     3=0.7
     4=0.7
     Index=&SpawnGroup()
     Num=::NewEnumerator3
    )
    Mesh Scale=&MergeHelper_VectorSwitch(
     0=&Multiply_VectorByFloat(Vector=&RandomRangeVector(Minimum=<1 0.5 0.5>, Maximum=<3 2 2>), Float=0.2)
     1=&Multiply_VectorByFloat(
      Vector=&RandomRangeVector(Minimum=<2.5 2 2>, Maximum=<3 2 2>)
      Float=&RandomRangeFloat(Minimum=0.03, Maximum=0.08)
     )
     2=&Multiply_VectorByFloat(Vector=&RandomRangeVector(Minimum=<2 0.5 0.5>, Maximum=<4 2 2>), Float=0.1)
     3=&Multiply_VectorByFloat(Vector=&RandomRangeVector(Minimum=<2.5 0.5 0.5>, Maximum=<5 2 2>), Float=0.08)
     4=&Multiply_VectorByFloat(Vector=&RandomRangeVector(Minimum=<2.5 0.5 0.5>, Maximum=<5 2 2>), Float=0.08)
     Index=&SpawnGroup()
     Num=::NewEnumerator3
    )
    Mesh Scale Mode=::NewEnumerator4
    UsePositionOffset=false
    Write Material Random=true
   )
   &SocketLocation(
    SamplePosOffset=true
    Sampled Position Offset=&MergeHelper_VectorSwitch(
     0=&RandomRangeVector(Minimum=<-16 82 3>, Maximum=<16 82 3>)
     1=<0 0 -8>
     2=&FN_FrontEndVector3Modifier(Gameplay=<0 -120 -14>, Front End=<0 -100 -15>)
     3=<20 -80 -12>
     4=<-20 -80 -12>
     Index=&SpawnGroup()
     Num=::NewEnumerator3
    )
    Socket Reader=$System.Socket Reader
    Socket Sampling Mode=::NewEnumerator3
    VelocitySpeedLimit=false
   )
   &TorusLocation(
    Align Torus With Axis=false
    Handle Radius=0
    Large Radius:NiagaraFloat=&MergeHelper_FloatSwitch(0=0, 1=25, 2=10, 3=7, 4=7, Index=&SpawnGroup(), Num=::NewEnumerator3)
    MaskSpawnGroup=false
    Module.Override Randomness=false
   )
   &AddVelocityFromPoint(
    Module.Use Falloff Distance=false
    Origin Offset=&MergeHelper_VectorSwitch(
     0=<0 -50 -20>
     1=<0 -50 -20>
     2=<0 -150 0>
     3=<0 -170 0>
     4=<0 -170 0>
     Index=&SpawnGroup()
     Num=::NewEnumerator3
    )
    Velocity Strength=&MergeHelper_FloatSwitch(
     0=0
     1=&RandomRangeFloat(Minimum=300, Maximum=350)
     2=200
     3=150
     4=150
     Index=&SpawnGroup()
     Num=::NewEnumerator3
    )
   ) {module=AddVelocityFromPoint001}
   &AddVelocity(
    Velocity=&MergeHelper_VectorSwitch(
     0=&FN_FrontEndVector3Modifier(
      Gameplay=&RandomRangeVector(Minimum=<-30 120 0>, Maximum=<30 200 130>)
      Front End=&RandomRangeVector(Minimum=<-30 120 0>, Maximum=<30 200 130>)
     )
     1=&RandomRangeVector(Minimum=<0 0 0>, Maximum=<0 0 300>)
     2=&FN_FrontEndVector3Modifier(Gameplay=<0 100 0>, Front End=<0 100 60>)
     3=<0 200 0>
     4=<0 200 0>
     Index=&SpawnGroup()
     Num=::NewEnumerator3
    )
    Velocity Speed Scale=&MergeHelper_FloatSwitch(
     0=&MergeHelper_FloatSwitch(
      0=&Lerp_Float(A=1, B=2, Alpha=$System.Timeline.Time)
      1=3
      2=?
      3=?
      Index=&SpawnGroup()
      Num=::NewEnumerator0
     )
     1=1
     2=1
     3=1
     4=1
     Index=&SpawnGroup()
     Num=::NewEnumerator3
    )
   )
   &AddVelocityInCone(
    Cone Angle=10
    Cone Axis=<0 1 0>
    Module.Use Velocity Falloff On Cone Axis=false
    Velocity Strength=&MergeHelper_FloatSwitch(0=0, 1=0, 2=100, 3=300, 4=300, Index=&SpawnGroup(), Num=::NewEnumerator3)
   ) {module=AddVelocityInCone001}
   &InitialMeshOrientation(
    Mesh Orientation Mode=::NewEnumerator2
    Rotation=&RandomRangeVector(Minimum=<0 0 0>, Maximum=<1 0 0>)
    Use Rotation Vector=true
   )
   &CameraOffset(Camera Offset Amount=&FN_FrontEndFloatModifier(Gameplay=-12, Front End=-30))
   &DynamicMaterialParameters(
    Index 0 Param 2=1
    Module.Param0WriteEnabled=true
    Module.Param1WriteEnabled=true
    Module.Param2WriteEnabled=true
    Module.Param3WriteEnabled=true
   )
   /SetVariables(Particles.Mesh_Low_Color:LinearColor=&MultiplyLinearColor(
     Linear Color=&MultiplyLinearColor(
      Linear Color=<rgba 0.7756 0.9185 1 1>
      Scale Factor=&MakeVector4_FromLinearColor(LinearColor=&MakeLinearColorFromVectorAndFloat(
        "Vector (RGB)"=&FN_FrontEndVector3Modifier(Gameplay=<1 1 1>, Front End=<0.4 0.8 1>)
        "Float (Alpha)"=1
       ))
     )
     Scale Factor=&MakeVector4_FromFloat(Float=&Multiply_Float(A=1, B=&FN_FrontEndFloatModifier(Gameplay=1, Front End=1)))
    )) {isSetParameters=true}
  ]
  particleUpdateScript [
   &GravityForce(Gravity=&MergeHelper_VectorSwitch(
     0=<0 0 0>
     1=<0 0 -1000>
     2=&Multiply_VectorByFloat(
      Vector=<0 0 3500>
      Float=&FloatFromCurve(FloatCurve=DataInterface() {propertyValues {Curve {keys [<at 0.1432 0 RCIM_Cubic>,<at 0.5915 1>]}}}, Scale Curve=0.2)
     )
     3=&Multiply_VectorByFloat(
      Vector=<0 0 3500>
      Float=&FloatFromCurve(FloatCurve=DataInterface() {propertyValues {Curve {keys [<at 0.1432 0 RCIM_Cubic>,<at 0.5915 1>]}}})
     )
     4=&Multiply_VectorByFloat(
      Vector=<0 0 3500>
      Float=&FloatFromCurve(FloatCurve=DataInterface() {propertyValues {Curve {keys [<at 0.1432 0 RCIM_Cubic>,<at 0.5915 1>]}}})
     )
     Index=&SpawnGroup()
     Num=::NewEnumerator3
    ))
   &ScaleVelocity(Velocity Scale=&MakeVector(
     X=1
     Y=1
     Z=&MergeHelper_FloatSwitch(
      0=&FloatFromCurve(FloatCurve=DataInterface() {propertyValues {Curve {keys [<at 0 0 RCIM_Cubic>,<at 1 1>]}}}, Scale Curve=1.5)
      1=1
      2=1
      3=&MakeFloatFromVector(
       Vector=&MakeVector(
        X=1
        Y=&FN_FrontEndFloatModifier(Gameplay=&ClampFloat(Float=$User.ForwardDot, Min=0.6), Front End=1)
        Z=1
       )
       Channel=::NewEnumerator0
      )
      4=&MakeFloatFromVector(
       Vector=&MakeVector(
        X=1
        Y=&FN_FrontEndFloatModifier(Gameplay=&ClampFloat(Float=$User.ForwardDot, Min=0.6), Front End=1)
        Z=1
       )
       Channel=::NewEnumerator0
      )
      Index=&SpawnGroup()
      Num=::NewEnumerator3
     )
    ))
   &SolveForcesAndVelocity(Clamp Velocity=false, Limit Acceleration=false)
   &ParticleState()
   &ScaleMeshSize(Scale Factor=&Multiply_VectorByFloat(
     Vector=&MergeHelper_VectorSwitch(0=<0.5 0.22 0.22>, 1=<1 1 1>, 2=<1 1 1>, 3=<1 1 1>, 4=<1 1 1>, Index=&SpawnGroup(), Num=::NewEnumerator3)
     Float=&MergeHelper_FloatSwitch(
      0=&FloatFromCurve(FloatCurve=DataInterface() {propertyValues {Curve {keys [
           {interpMode=RCIM_Cubic, tangentMode=RCTM_User, time=0, value=0}
           <at 0.2251 1 RCIM_Cubic>
           {arriveTangent=-1.0223, interpMode=RCIM_Cubic, leaveTangent=-1.0223, tangentMode=RCTM_User, time=1, value=0}
          ]}}})
      1=&FloatFromCurve(FloatCurve=DataInterface() {propertyValues {Curve {keys [
           {interpMode=RCIM_Cubic, tangentMode=RCTM_User, time=0, value=0}
           <at 0.2 0.95>
           {arriveTangent=-1.0223, interpMode=RCIM_Cubic, leaveTangent=-1.0223, tangentMode=RCTM_User, time=1, value=0}
          ]}}})
      2=&FloatFromCurve(FloatCurve=DataInterface() {propertyValues {Curve {keys [
           {interpMode=RCIM_Cubic, tangentMode=RCTM_User, time=0, value=0}
           <at 0.15 0.95>
           {arriveTangent=-1.0223, interpMode=RCIM_Cubic, leaveTangent=-1.0223, tangentMode=RCTM_User, time=1, value=0}
          ]}}})
      3=&FloatFromCurve(FloatCurve=DataInterface() {propertyValues {Curve {keys [
           {interpMode=RCIM_Cubic, tangentMode=RCTM_User, time=0, value=0}
           <at 0.15 0.95>
           {arriveTangent=-1.0223, interpMode=RCIM_Cubic, leaveTangent=-1.0223, tangentMode=RCTM_User, time=1, value=0}
          ]}}})
      4=&FloatFromCurve(FloatCurve=DataInterface() {propertyValues {Curve {keys [
           {interpMode=RCIM_Cubic, tangentMode=RCTM_User, time=0, value=0}
           <at 0.15 0.95>
           {arriveTangent=-1.0223, interpMode=RCIM_Cubic, leaveTangent=-1.0223, tangentMode=RCTM_User, time=1, value=0}
          ]}}})
      Index=&SpawnGroup()
      Num=::NewEnumerator3
     )
    ))
  ]
  rendererClasses [&NiagaraMeshRendererProperties]
  renderers [
   {rendererClass=&NiagaraMeshRendererProperties, rendererIndex=0}
   {rendererClass=&NiagaraMeshRendererProperties, rendererIndex=1}
  ]
 }
 TrailWings {
  emitterSpawnScript [&FN_RibbonIDOffsetUtility()]
  emitterUpdateScript [
   &EmitterState()
   &FN_SpawnAnimTrail()
  ]
  particleSpawnScript [
   &FN_InitializeAnimTrailBoneIndecies(First Bone or Socket Index=1, Second Bone or Socket Index=2)
   &FN_InitializeAnimTrail(
    AnimTrailWidth=&FN_FrontEndFloatModifier(Gameplay=3.6, Front End=2.7)
    Color=<rgba 0.7917 0.7917 0.7917 1.2>
    Lifetime=&FN_FrontEndFloatModifier(Gameplay=&Lerp_Float(A=0.07, B=0.22, Alpha=$System.Timeline.Time), Front End=0.2)
   )
   &AddVelocity(
    Velocity=&FN_FrontEndVector3Modifier(Gameplay=<0 2500 70>, Front End=<0 2500 70>)
    Velocity Speed Scale=&FN_FrontEndFloatModifier(Gameplay=&Lerp_Float(A=0, B=0.6, Alpha=$System.Timeline.Time), Front End=0.5)
   )
  ]
  particleUpdateScript [
   &GravityForce(Gravity=<0 0 1300>)
   &SolveForcesAndVelocity(Clamp Velocity=false, Limit Acceleration=false)
   &FN_UpdateAnimTrail()
   &InterpolateOverTime(Initial Float Value:NiagaraFloat=$System.Timeline.Time, Target Float Value=1, Use Set Interpolation Rate=true)
   &KillParticles(Kill Particles=&SetBoolByFloatComparison(
     A=&FN_FrontEndFloatModifier(Gameplay=$System.Timeline.Time, Front End=1)
     B=$Particles.InterpolateOverTime.InterpolateOverTime_Float.Moving Average
     Comparison Type=::NewEnumerator6
    ))
   &ScaleColor(
    Scale Alpha=&FNTimeOfDayScalar(Night Scale=1, Morning Scale=1, Evening Scale=1)
    Scale RGB=&MakeVectorFromLinearColorRGB(Color=$System.TOD Color)
    ScaleA=true
    ScaleRGB=true
   )
   &ScaleColor(
    Scale Alpha=&FN_FrontEndFloatModifier(Gameplay=&Lerp_Float(A=0, B=1, Alpha=$System.Timeline.Time), Front End=1)
    ScaleA=true
    ScaleRGB=true
   ) {module=ScaleColor003}
   &ScaleColor(
    Scale Alpha=&FN_FrontEndFloatModifier(
     Gameplay=&FloatFromCurve(FloatCurve=DataInterface() {propertyValues {Curve {keys [<at 2.8 0>,<at 3.2 1>]}}}, CurveIndex=$System.Age)
     Front End=&FloatFromCurve(FloatCurve=DataInterface() {propertyValues {Curve {keys [<at 3.3 0>,<at 4 1>]}}}, CurveIndex=$System.Age)
    )
    ScaleA=true
    ScaleRGB=true
   ) {module=ScaleColor001}
   &ScaleRibbonWidth(Ribbon Width Scale=&FloatFromCurve(FloatCurve=DataInterface() {propertyValues {Curve {keys [<at 0.1167 1 RCIM_Cubic>,<at 0.6828 0.332>]}}}))
  ]
  rendererClasses [&NiagaraRibbonRendererProperties]
  renderers [{rendererClass=&NiagaraRibbonRendererProperties}]
 }
 WindLines {
  emitterUpdateScript [
   &EmitterState(Delay First Loop Only=true, Life Cycle Mode=::NewEnumerator1, Loop Delay=2.8, Loop Duration=1, UseLoopDelay=true)
   &SpawnRate(
    SpawnRate=&Lerp_Float(
     A=50
     B=150
     Alpha=&RemapRange(Input Value To Remap=$User.RightDot, Input Min=-1, Input Max=1, Output Min=0, Output Max=1)
    )
    Use Spawn Probability=true
   ) {enabled=false}
   &SpawnRate(
    Spawn Probability=&FN_FrontEndFloatModifier(Gameplay=$System.Timeline.Time, Front End=1)
    SpawnGroup=1
    SpawnRate=60
    Use Spawn Probability=true
   ) {module=SpawnRate001}
  ]
  particleSpawnScript [
   &InitializeParticle(Mesh Scale Max=<25 1 1>, Mesh Scale Min=<10 0.5 0.5>, Mesh Scale Mode=::NewEnumerator2, UsePositionOffset=false, Write Material Random=true)
   &InitialMeshOrientation(
    Mesh Orientation Mode=::NewEnumerator2
    Rotation=&RandomRangeVector(Minimum=<0 0 0>, Maximum=<1 0 0>)
    Use Rotation Vector=true
   )
   &TorusLocation(
    Align Torus With Axis=true
    Handle Radius=0
    Large Radius:NiagaraFloat=&RandomRangeFloat(Minimum=80, Maximum=150)
    MaskSpawnGroup=false
    Offset=<0 -180 0>
    Torus Axis=<0 1 0>
    Torus Distribution Mode:ENiagaraTorusDistributionMode=::NewEnumerator1
    Torus Origin:NiagaraPosition=&Add_Vector(
     A=Linked() {linkedVariable {name=Particles.Position, type=NiagaraPosition}}
     B=&MakeVector(X=0, Y=&MakeCustomFloatFromBool(Bool=$User.bIsFrontEndPreview, True Float=200, False Float=0), Z=0)
    )
    U Position:NiagaraFloat=&MakeCustomFloatFromBool(
     Bool=&SetBoolByFloatComparison(A=$User.RightDot, B=0, Comparison Type=::NewEnumerator4)
     True Float=&RandomRangeFloat(Minimum=0, Maximum=1)
     False Float=&Lerp_Float(
      A=&RandomRangeFloat(Minimum=0.25, Maximum=0.75)
      B=&RandomRangeFloat(Minimum=-0.25, Maximum=0.25)
      Alpha=&RemapRange(Input Value To Remap=$User.RightDot, Input Min=-1, Input Max=1, Output Min=0, Output Max=1)
     )
    )
    V Position:NiagaraFloat=&RandomRangeFloat(Minimum=0, Maximum=1)
   ) {enabled=false}
   &TorusLocation(
    Align Torus With Axis=true
    Handle Radius=0
    Large Radius:NiagaraFloat=&RandomRangeFloat(Minimum=100, Maximum=130)
    MaskSpawnGroup=false
    Module.Override Randomness=false
    Offset=<0 -180 0>
    Spawn Group Mask=1
    Torus Axis=<0 1 0>
    Torus Origin:NiagaraPosition=&Add_Vector(A=<0 -100 0>, B=&Lerp_Vector(A=<0 0 0>, B=<200 0 0>, Alpha=$User.RightDot))
   ) {module=TorusLocation001}
   &AddVelocityFromPoint(
    Module.Use Falloff Distance=false
    Origin Offset=&Add_Vector(A=<0 -500 0>, B=&Lerp_Vector(A=<0 0 0>, B=<300 0 0>, Alpha=$User.RightDot))
    Velocity Strength=300
   )
   &AddVelocity(
    Module.Use Falloff Distance=false
    Origin Offset=&Add_Vector(A=<0 -500 0>, B=&Lerp_Vector(A=<0 0 0>, B=<500 0 0>, Alpha=$User.RightDot))
    Override Seed=false
    Velocity Mode=::NewEnumerator1
    Velocity Speed=400
   ) {module=AddVelocity001}
   &AddVelocity(Velocity=&Multiply_Vector(
     A=&RandomRangeVector(Minimum=<1 200 1>, Maximum=<1 300 1>)
     B=&MakeVector(X=1, Y=&MakeCustomFloatFromBool(Bool=$User.bIsFrontEndPreview, True Float=-1, False Float=1), Z=1)
    ))
  ]
  particleUpdateScript [
   &InheritSourceMovement(Applied Position Scale=<0.9 0.9 1>, ApplyForce=false, ApplyPosition=true, ApplyPositionFalloff=false) {enabled=false}
   &ScaleVelocity(Velocity Scale=&VectorFromFloat(Value=&FloatFromCurve(
      FloatCurve=DataInterface() {propertyValues {Curve {keys [
          {interpMode=RCIM_Cubic, tangentMode=RCTM_User, time=0, value=0}
          {arriveTangent=2, interpMode=RCIM_Cubic, leaveTangent=2, time=0.1, value=0.975}
          {interpMode=RCIM_Cubic, tangentMode=RCTM_User, time=0.5, value=1}
          {arriveTangent=-2, interpMode=RCIM_Cubic, leaveTangent=-2, time=0.9, value=1}
          {tangentMode=RCTM_User, time=1, value=0}
         ]}}}
      CurveIndex=&RemapRange(Input Value To Remap=$User.RightDot, Input Min=-1, Input Max=1, Output Min=0, Output Max=1)
      Scale Curve=&FN_FrontEndFloatModifier(Gameplay=1, Front End=4)
     )))
   &SolveForcesAndVelocity(Clamp Velocity=false, Limit Acceleration=false)
   &ParticleState()
   &ScaleMeshSize(Scale Factor=&VectorFromFloat(Value=&FloatFromCurve(FloatCurve=DataInterface() {propertyValues {Curve {keys [
          {tangentMode=RCTM_User, time=0, value=1}
          {arriveTangent=-0.0405, interpMode=RCIM_Cubic, leaveTangent=-0.0405, tangentMode=RCTM_User, time=0.7009, value=1}
          {arriveTangent=-8.0089, interpMode=RCIM_Cubic, leaveTangent=-8.0089, tangentMode=RCTM_User, time=1, value=0}
         ]}}})))
   &DynamicMaterialParameters(
    Index 0 Param 2=2
    Index 0 Param 3=1
    Index 0 Param 4=1
    Module.Index0Float0Write=true
    Module.Index0Float1Write=true
    Module.Index0Float2Write=true
    Module.Index0Float3Write=true
   )
   &ScaleColor(
    Scale Alpha=&FloatFromCurve(
     FloatCurve=DataInterface() {propertyValues {Curve {keys [
         {arriveTangent=3.2319, interpMode=RCIM_Cubic, leaveTangent=3.2319, tangentMode=RCTM_User, time=0, value=0}
         {interpMode=RCIM_Cubic, tangentMode=RCTM_User, time=0.5, value=2}
         {arriveTangent=-3.6409, tangentMode=RCTM_User, time=1, value=0}
        ]}}}
     Scale Curve=&FN_FrontEndFloatModifier(
      Gameplay=&Lerp_Float(A=0, B=&Lerp_Float(A=0, B=1, Alpha=$System.Timeline001.Time), Alpha=$System.Timeline.Time)
      Front End=&Multiply_Float(A=$System.Timeline001.Time, B=0.9)
     )
    )
    ScaleA=true
    ScaleRGB=true
   )
   &Color(Color=&ColorFromCurve(
     ColorCurve=DataInterface() {propertyValues {
       AlphaCurve {keys [<at 0 0 RCIM_Cubic>,<at 0.0502 1>,<at 0.4858 1>,<at 0.9519 1>,<at 1 0 RCIM_Cubic>]}
       BlueCurve {keys [<at 0 1 RCIM_Cubic>,<at 0.4934 0.0763>,<at 1 0 RCIM_Cubic>]}
       GreenCurve {keys [<at 0 0.0001 RCIM_Cubic>,<at 0.4934 0>,<at 1 1 RCIM_Cubic>]}
       RedCurve {keys [<at 0 0 RCIM_Cubic>,<at 0.4934 1>,<at 1 0.0424 RCIM_Cubic>]}
      }}
     CurveIndex=&RemapRange(Input Value To Remap=$User.RightDot, Input Min=-1, Input Max=1, Output Min=0, Output Max=1)
    )) {enabled=false}
  ]
  rendererClasses [&NiagaraMeshRendererProperties, &NiagaraSpriteRendererProperties]
  renderers [
   {rendererClass=&NiagaraMeshRendererProperties, rendererIndex=0}
   {rendererClass=&NiagaraSpriteRendererProperties, rendererIndex=1}
  ]
 }
}
summary {
 systemName=NS_Glider_CozyMechanic_Trail_01
 emitters [
  {emitterName=WindLines, rendererClasses [&NiagaraMeshRendererProperties, &NiagaraSpriteRendererProperties]}
  {emitterName=TrailWings, rendererClasses [&NiagaraRibbonRendererProperties]}
  {emitterName=Frost, rendererClasses [&NiagaraSpriteRendererProperties]}
  {emitterName=Snow_Idle, rendererClasses [&NiagaraMeshRendererProperties]}
  {emitterName=CPU_particles, rendererClasses [&NiagaraSpriteRendererProperties]}
  {emitterName=LongTrail, rendererClasses [&NiagaraMeshRendererProperties]}
 ]
 userVariables [
  User.bIsFrontEnd {type=NiagaraBool}
  User.bIsFrontEndPreview {type=NiagaraBool}
  User.bIsFullyDeployed {type=NiagaraBool}
  User.ForwardDot {type=NiagaraFloat, defaultValue=0, description=""}
  User.RibbonIDOffset {type=NiagaraInt32}
  User.RightDot {type=NiagaraFloat, defaultValue=0, description=""}
 ]
}
